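; Tornado actor for the HighTower Squash TornadoJump item that launches overlapping pawns and vehicles unless their tags are ignored.
; Collision enables after a spawn delay; a lifetime timer sets replicated Hide, whose OnRep plays despawn effects.
; Launch direction follows pawn acceleration or other actors' velocity, scaled by data rows and gravity attributes.
; Pawns gain fall-damage immunity; vehicles are pushed through BroadcastAbilityImpactatLocation, which a note says needs its 780 radius.
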(event Custom|LifeEnded
  ;[1 1 1] On life ended set hide. Set lifespan and leave it to die.
  (if (Networking|IsServer)
    (Variables|Default|SetHide true)
    (Networking|ForceNetUpdate)
    (Actor|SetLifeSpan (Variables|Default|GetLifeAfterDeath))))

(event Custom|TurnOnCollision
  ;[1 1 1] Turn on the collision a moment after spawn so that the tornado has time to build up before it will knock people away
  (Collision|SetActorEnableCollision true))

(event EventBeginPlay
  ;[1 1 1] Begin Play - Spawn FX and set lifespan up.
  (CallFunction|SpawnFX)
  (bind _returnvalue (Utilities|Time|SetTimerbyEvent (EventDispatchers|CreateEvent :!event Custom|LifeEnded) (ScalableFloat|GetValueatLevel (Variables|Default|GetRow_Lifetime))))
  (Variables|Default|SetTimer_LifeEnded _returnvalue)
  ;[1 1 1] Turn on the collision a moment after spawn so that the tornado has time to build up before it will knock people away
  (bind _returnvalue_2 (Utilities|Time|SetTimerbyEvent (delegate Custom|TurnOnCollision) (Variables|Default|GetTime_TurnOnCollision)))
  (Variables|Default|SetTimer_TurnOnCollision _returnvalue_2))

(event Custom|Spawn%20FX
  ;[1 1 1] FX and Sounds
  ;;[1 1 1] Spawn FX - Handle visuals for the tornado spawning in
  (Ability|ExecuteGameplayCueReplicated (Variables|Default|GetGC_Spawn)))

(event Custom|PlayDespawn
  ;[1 1 1] FX and Sounds
  ;;[1 1 1] Play Despawn - Play GC, Hide the actor, turn off collision and fade out the audio.
  (Ability|ExecuteGameplayCueReplicated (Variables|Default|GetGC_Despawn))
  (Rendering|SetActorHiddenInGame true)
  (Collision|SetActorEnableCollision false)
  (Audio|Components|Audio|FadeOut (Variables|Default|GetAmbientAudio) (Variables|Default|GetAudioFadeOutDuration) (Variables|Default|GetAudioFadeVolumeLevel) "Linear" :!class /Script/Engine.AudioComponent))

(event OnComponentBeginOverlap(TornadoMesh)
  (params (OverlappedComponent :type Primitive%20Component%20Object%20Reference) (OtherActor :type Actor%20Object%20Reference) (OtherComp :type Primitive%20Component%20Object%20Reference) (OtherBodyIndex :type Integer) (bFromSweep :type Boolean) (SweepResult :type Hit%20Result%20Structure%20%28by%20ref%29))
  ;[1 1 1] If a Player Overlaps Launch them
  (bind _returnvalue (GameplayTags|HasAnyMatchingGameplayTags(Message) OtherActor (Variables|Default|GetTC_TagsToIgnore)))
  (if (not _returnvalue)
    (Development|PrintString (Utilities|String|Append "Tornado Launch: " (Utilities|GetDisplayName OtherActor)))
    (CallFunction|Launch OtherActor)))

(event Custom|Launch
  (params (OverlappedActor :type Actor%20Object%20Reference))
  ;[1 1 1] Launch the overlapped Actor, doing some calculations based on rows and also playing a GC.
  (Ability|ExecuteGameplayCueWithParamsReplicated (Variables|Default|GetGC_Launch) (Ability|GameplayCue|MakeGameplayCueParameters :NormalizedMagnitude 0.0 :RawMagnitude 0.0 :Location (Transformation|GetActorLocation OverlappedActor) :Normal (Transformation|GetActorForwardVector OverlappedActor) :Instigator OverlappedActor :EffectCauser OverlappedActor :SourceObject OverlappedActor :GameplayEffectLevel 1 :AbilityLevel 1 :bReplicateLocationWhenUsingMinimalRepProxy false))
  (bind _launchmag (CallFunction|GetInitialKnockMag OverlappedActor))
  (bind _asfort_pawn (Utilities|Casting|CastToFortPawn OverlappedActor)
    (:then
      ;;[1 1 1] Launch Pawns
      (bind _launchedpawn (Utilities|LaunchPawns (Actor|GetOwner) _asfort_pawn (* _launchmag (CallFunction|CheckGravityScale _asfort_pawn "(AttributeName=\"GravityZScale\",Attribute=/Script/FortniteGame.FortMovementSet:GravityZScale,AttributeOwner=None)")) true true false true true false true true))
      ;;[1 1 1] Give Fall Immunity
      (Utilities|IsValid (Ability|GetFortAbilitySystemComponent _launchedpawn)
        (:Is%20Valid
          (GameplayEffects|ApplyGameplayEffectToTarget :self (Ability|GetFortAbilitySystemComponent self) :GameplayEffectClass (Variables|Default|GetGE_LaunchFallDamageImmune) :Target (Ability|GetFortAbilitySystemComponent _launchedpawn) :Level 0.0))))
    (:CastFailed
      ;;[1 1 1] Vehicle Launch
      (bind _asfort_vehicle_interface (Utilities|Casting|CastToFortVehicleInterface OverlappedActor))
      (bind _returnvalue (Vehicle|GetVehicleActor _asfort_vehicle_interface :!class /Script/FortniteGame.FortVehicleInterface))
      (Audio|BroadcastAbilityImpactatLocation "/Game/Athena/Sounds/Events/Cube/Cube_LaunchPlayer_Cue.Cube_LaunchPlayer_Cue" _returnvalue (* (Math|Vector|Normalize (CallFunction|CalcVehicleMult _launchmag)) (* (CallFunction|CheckGravityScale OverlappedActor "(AttributeName=\"VehicleGravityScale\",Attribute=/Script/FortniteGame.FortVehicleMovementSet:VehicleGravityScale,AttributeOwner=None)") (Math|Vector|VectorLength (CallFunction|CalcVehicleMult _launchmag)))) 7.8e+02) ; This is the secret node that launches vehicles. Note it needs the 780 radius to work.
    )) ; Check if pawn or not
)

(fn CalcVehicleMult
  (params (InVec :type Vector))
  ;[1 1 1] Calculate the impulse amount the vehicle should get.
  (return (Math|Vector|MakeVector (* (.x InVec) (ScalableFloat|GetValueatLevel (Variables|Default|GetRow_VehicleXYMult))) (* (.y InVec) (ScalableFloat|GetValueatLevel (Variables|Default|GetRow_VehicleXYMult))) (* (.z InVec) (ScalableFloat|GetValueatLevel (Variables|Default|GetRow_VehicleZMult))))))

(fn CheckGravityScale
  (params (Actor :type Actor%20Object%20Reference) (AttributeGravityScale :type Gameplay%20Attribute%20Structure))
  (bind (_bsuccessfullyfoundattribute _returnvalue) (Ability|Attribute|GetFloatAttribute Actor AttributeGravityScale))
  ;[1 1 1] Get the gravity scale Attribute and calc the gravity modifier. If it is more than 1.0 then try to increase the power and if it is less than 1.0 decrease the power
  (if _bsuccessfullyfoundattribute
    (if (< _returnvalue 1.0)
      (return (Math|Float|SafeDivide _returnvalue (Variables|Default|GetLowZGravityDampenerIntensity)))
      (elif (> _returnvalue 1.0)
        (return (* _returnvalue (Variables|Default|GetHighZGravityMultiplierIntensity)))))
    (else
      _))
  (return 1.0))

(fn GetInitialKnockMag
  (params (Target :type Actor%20Object%20Reference))
  (bind (_returnvalue_x _returnvalue_y _returnvalue_z) (Math|Vector|BreakVector (Math|Vector|Normalize (Math|Vector|SelectVector (Math|Vector|Normalize (Pawn|Components|CharacterMovement|GetCurrentAcceleration (Variables|/Script/Engine.Character|GetCharacterMovement (Utilities|Casting|CastToFortPawn Target)))) (Math|Vector|Normalize (Transformation|GetVelocity Target)) (.bSuccess (Utilities|Casting|CastToFortPawn Target))))))
  (bind (_asfort_pawn _bsuccess) (Utilities|Casting|CastToFortPawn Target))
  ;[1 1 1] Base knock direction on the players acceleration and if they are not a pawn their velocity. Then use rows to find the mag and return the launch vector.
  (return (Math|Vector|MakeVector (.x (Math|Vector|MakeVector (* _returnvalue_x (ScalableFloat|GetValueatLevel (Variables|Default|GetRow_XYKnockMag))) (* _returnvalue_y (ScalableFloat|GetValueatLevel (Variables|Default|GetRow_XYKnockMag))) (* (Math|Float|Clamp(Float) _returnvalue_z) (ScalableFloat|GetValueatLevel (Variables|Default|GetRow_ZKnockMag))))) (.y (Math|Vector|MakeVector (* _returnvalue_x (ScalableFloat|GetValueatLevel (Variables|Default|GetRow_XYKnockMag))) (* _returnvalue_y (ScalableFloat|GetValueatLevel (Variables|Default|GetRow_XYKnockMag))) (* (Math|Float|Clamp(Float) _returnvalue_z) (ScalableFloat|GetValueatLevel (Variables|Default|GetRow_ZKnockMag))))) (+ (.z (Math|Vector|MakeVector (* _returnvalue_x (ScalableFloat|GetValueatLevel (Variables|Default|GetRow_XYKnockMag))) (* _returnvalue_y (ScalableFloat|GetValueatLevel (Variables|Default|GetRow_XYKnockMag))) (* (Math|Float|Clamp(Float) _returnvalue_z) (ScalableFloat|GetValueatLevel (Variables|Default|GetRow_ZKnockMag))))) (ScalableFloat|GetValueatLevel (Variables|Default|GetRow_ZAdd))))))

(fn OnRep_Hide
  (CallFunction|PlayDespawn))

(fn ConstructionScript
  (Collision|SetActorEnableCollision false))
