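; Gameplay ability for the HighTower Squash Tornado Jump that extends a parent leap ability and spawns a tornado gameplay actor.
; On leap, it calls Target, spawns the tornado building actor at the player's location and rotation, then runs the parent leap.
; Cast montage notifies matching the JumpLoop tag are resent as gameplay events, while others go to the parent.
; Overrides place the arc point above the player by vertical launch magnitude and offset the targeting source.
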(event Custom|SpawnTornado
  (params (Location :type Vector) (Rotation :type Rotator))
  ;[1 1 1] SpawnTornado
  (Building|SpawnBuildingGameplayActor (Variables|Default|GetBGA_Tornado) (Math|Transform|MakeTransform Location Rotation) (Variables|Default|GetPlayerPawn)))

(event EventLeap
  ;[1 1 1] On Leap Target and break builds / knock back players.
  (CallFunction|Target)
  (CallFunction|SpawnTornado (Transformation|GetActorLocation (Variables|Default|GetPlayerPawn)) (Transformation|GetActorRotation (Variables|Default|GetPlayerPawn)))
  (|CallParentFunction))

(event EventCastMontageNotifyStart
  (params (NotifyTag :type Gameplay%20Tag%20Structure))
  (bind _event_jumploop (Variables|Default|GetEvent_JumpLoop))
  ;[1 1 1] If the notify is the looping jump send that event otherwise do the normal path.
  (if (GameplayTags|Equal(GameplayTag) NotifyTag _event_jumploop)
    (Ability|SendGameplayEvent _event_jumploop)
    (else
      (|CallParentFunction NotifyTag))))

(fn GetArcCustomLocation
  (return (Math|Vector|vector+vector (Transformation|GetActorLocation (Variables|Default|GetPlayerPawn)) (Math|Vector|MakeVector 0.0 0.0 (ScalableFloat|GetValueatLevel (Variables|PowerLeap|UpwardorDownward|GetRow_VerticalLaunchMagnitude))))))

(fn GetCustomAbilitySourceTransform
  ;[1 1 1] Take the player's location and shift it up a little for the targeting location
  (return (Math|Transform|MakeTransform (+ (Transformation|GetActorLocation (Ability|GetAvatarActorfromActorInfo)) (Variables|Default|GetJumpTargetOffset)) (Transformation|GetActorRotation (Ability|GetAvatarActorfromActorInfo)))))
